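# Sound cue for equipping the Bush consumable item.
# A Random node with equal weights picks one of two equip wave variations and feeds a Modulator that varies pitch on each play.
# It uses weapon-drop attenuation and the loot-collect sound class, and its random node is excluded from branch culling.
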SoundCue /Game/Athena/Sounds/Items/Bush/Item_Consumables_Bush_Equip_Item_Cue
root=Modulator_2
nodes [
 {id=Modulator_2, children [Random_2], properties {pitchMax=1.1}}
 {id=Random_2, children [WavePlayer_0,WavePlayer_7], properties {bShouldExcludeFromBranchCulling=true, weights [1,1]}}
 {id=WavePlayer_0, properties {soundWaveAssetPtr=&/Bush_Equip_01}}
 {id=WavePlayer_7, properties {soundWaveAssetPtr=&/Bush_Equip_02}}
]
properties {
 attenuationSettings=&Fort_Weapon_Drop_Att
 maxDistance=3000
 soundClassObject=&Fort_Loot_Collect
 volumeMultiplier=1
}
